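; BP_SteamyStacks_SmokePlume1 is a Blueprint, apparently for a smoke plume effect at Steamy Stacks, with a replicated Visibility variable.
; OnRep_Visibility switches on that integer and calls the Visible custom event when the value is 1.
; BeginPlay, the construction script and both the Visible and Invisible events show no logic here, and Invisible is not called in this file.
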(event EventBeginPlay)

(event Custom|Visible)

(event Custom|Invisible)

(fn OnRep_Visibility
  (switch int (Variables|Default|GetVisibility)
    (:1
      (CallFunction|Visible))))

(fn ConstructionScript)
